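; Animated wild-west rocking chair prop that rocks and plays audio when a Fortnite player pawn enters its trigger box.
; A DoOnce gates the trigger, with the authority running a timeline that lerps between cached start and end rotations and remotes playing the sound.
; On its own death or that of the building actor it attaches to, DestroyedLocally hides the mesh and plays destruction audio and effects on remotes.
; In the attachment case, the actor is destroyed after a short delay.
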(event EventBeginPlay
  (bind _wildwest_rockingchair_1 (Variables|Default|GetWildWest_RockingChair_1))
  (bind _relativerotation (Variables|/Script/Engine.SceneComponent|GetRelativeRotation _wildwest_rockingchair_1))
  (Variables|Default|SetChairStartRelRot _relativerotation)
  (Variables|Default|SetChairEndRelRot (Math|Rotator|CombineRotators _relativerotation "0.000000,0.000000,30.000000"))
  (Networking|SwitchHasAuthority
    (:Authority
      (bind _asbuilding_actor (Utilities|Casting|CastToBuildingActor (Variables|Default|GetAttachmentActor)))
      (EventDispatchers|BindEventtoOnDied _asbuilding_actor (delegate Custom|AttachmentPropDestroyed)))))

(event Custom|AttachmentPropDestroyed
  (params (DamagedActor :type Actor%20Object%20Reference) (Damage :type Float%20%28single-precision%29) (InstigatedBy :type Controller%20Object%20Reference) (DamageCauser :type Actor%20Object%20Reference) (HitLocation :type Vector) (FHitComponent :type Primitive%20Component%20Object%20Reference) (BoneName :type Name) (Momentum :type Vector))
  (bind _self self)
  (CallFunction|DestroyedLocally _self)
  (Utilities|FlowControl|Delay 0.2)
  (Actor|DestroyActor _self))

(event OnComponentBeginOverlap(TriggerBox)
  (params (OverlappedComponent :type Primitive%20Component%20Object%20Reference) (OtherActor :type Actor%20Object%20Reference) (OtherComp :type Primitive%20Component%20Object%20Reference) (OtherBodyIndex :type Integer) (bFromSweep :type Boolean) (SweepResult :type Hit%20Result%20Structure%20%28by%20ref%29))
  (Utilities|Casting|CastToFortPlayerPawnAthena OtherActor)
  (Utilities|FlowControl|DoOnce false
    (:Completed
      (Networking|SwitchHasAuthority
        (:Authority
          (bind (_direction _alpha) (Timeline|Anim PlayFromStart)
            (:Update
              (Transformation|SetRelativeRotation (Variables|Default|GetWildWest_RockingChair_1) (Math|Rotator|Lerp(Rotator) (Variables|Default|GetChairStartRelRot) (Variables|Default|GetChairEndRelRot) _alpha true) false false))
            (:Finished
              (Utilities|FlowControl|Delay 2.0))))
        (:Remote
          (Audio|Components|Audio|Play (Variables|Default|GetAudio) 0.0))))))

(event EventOnDeathServer
  (params (Damage :type Float%20%28single-precision%29) (DamageTags :type Gameplay%20Tag%20Container%20Structure%20%28by%20ref%29) (Momentum :type Vector :default "0, 0, 0") (HitInfo :type Hit%20Result%20Structure%20%28by%20ref%29) (InstigatedBy :type Controller%20Object%20Reference) (DamageCauser :type Actor%20Object%20Reference) (EffectContext :type Gameplay%20Effect%20Context%20Handle%20Structure))
  (CallFunction|DestroyedLocally))

(event Custom|DestroyedLocally
  (Utilities|FlowControl|DoOnce false
    (:Completed
      (Networking|SwitchHasAuthority
        (:Remote
          (Rendering|SetVisibility (Variables|Default|GetWildWest_RockingChair_1) false false)
          (Audio|Components|Audio|Play (Variables|Default|GetDestructionAudio) 0.0)
          (Components|Activation|Activate (Variables|Default|GetP_Destruction_HauntedHouse) false))))))

(fn ConstructionScript)
